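# Physics asset for the Gumball backpack cosmetic, defining the bodies and joints that let its bag, flaps, ooze strands and balls move as secondary motion.
# A master bag body is angularly locked to the pack and serves as the hub for the bag, two flaps and four balls.
# The bag and flaps have narrow swing limits, while the four ball spheres rotate freely.
# Four ooze capsules hang off the bag with limited swing and locked twist.
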PhysicsAsset /BRCosmetics/Accessories/FORT_Backpacks/Backpack_M_MED_Gumball/Meshes/M_MED_Gumball_Pack_PhysicsAsset
bodies {
 dyn_bag [{shapeType=Capsule, length=3, radius=3, shapeName=dyn_bag_capsule, center=<1.5 0 0>, rotation=<rot 90 0 0>}]
 dyn_bag_MSTR [{shapeType=Sphere, radius=1, shapeName=dyn_bag_MSTR_sphere, center=<0.2527 1.2465 5.429>}]
 dyn_ball_01 [{shapeType=Sphere, radius=1.1804, shapeName=dyn_ball_01_sphere, center=<0.0189 0.0088 -0.0002>}]
 dyn_ball_02 [{shapeType=Sphere, radius=1.4337, shapeName=dyn_ball_02_sphere, center=<-0.0121 0.0008 -0.033>}]
 dyn_ball_03 [{shapeType=Sphere, radius=1.6085, shapeName=dyn_ball_03_sphere, center=<-0.0095 0.0014 -0.0223>}]
 dyn_ball_04 [{shapeType=Sphere, radius=1.9034, shapeName=dyn_ball_04_sphere, center=<-0.0065 -0.0037 -0.0382>}]
 dyn_flap_01 [{shapeType=Capsule, length=3, radius=3, shapeName=dyn_flap_01_capsule, center=<2.4205 -0.0001 0>, rotation=<rot 90 0 0>}]
 dyn_flap_02 [{shapeType=Capsule, length=3, radius=3, shapeName=dyn_flap_02_capsule, center=<1 0 0>, rotation=<rot 90 0 0>}]
 dyn_ooze_01 [{shapeType=Capsule, length=3, radius=2, shapeName=dyn_ooze_01_capsule, center=<1 0 0>, rotation=<rot 90 0 0>}]
 dyn_ooze_02 [{shapeType=Capsule, length=3, radius=2, shapeName=dyn_ooze_02_capsule, center=<1 0 0>, rotation=<rot 90 0 0>}]
 dyn_ooze_03 [{shapeType=Capsule, length=3, radius=2, shapeName=dyn_ooze_03_capsule, center=<1 0 0>, rotation=<rot 90 0 0>}]
 dyn_ooze_04 [{shapeType=Capsule, length=3, radius=2, shapeName=dyn_ooze_04_capsule, center=<1 0 0>, rotation=<rot 90 0 0>}]
 pack [{shapeType=Capsule, length=3, radius=3, shapeName=pack_capsule}]
}
constraints [
 {
  bone1Name=dyn_bag_MSTR
  bone2Name=pack
  swing1LimitDegrees=45
  swing1Motion=Locked
  swing2LimitDegrees=45
  swing2Motion=Locked
  twistLimitDegrees=45
  twistMotion=Locked
 }
 {bone1Name=dyn_bag, bone2Name=dyn_bag_MSTR, swing1LimitDegrees=10, swing2LimitDegrees=20, twistLimitDegrees=1, twistMotion=Locked}
 {bone1Name=dyn_ooze_01, bone2Name=dyn_bag, swing1LimitDegrees=30, swing2LimitDegrees=30, twistLimitDegrees=1, twistMotion=Locked}
 {bone1Name=dyn_ooze_02, bone2Name=dyn_bag, swing1LimitDegrees=30, swing2LimitDegrees=30, twistLimitDegrees=1, twistMotion=Locked}
 {bone1Name=dyn_ooze_03, bone2Name=dyn_bag, swing1LimitDegrees=30, swing2LimitDegrees=30, twistLimitDegrees=1, twistMotion=Locked}
 {bone1Name=dyn_ooze_04, bone2Name=dyn_bag, swing1LimitDegrees=30, swing2LimitDegrees=30, twistLimitDegrees=1, twistMotion=Locked}
 {bone1Name=dyn_flap_01, bone2Name=dyn_bag_MSTR, swing1LimitDegrees=1, swing2LimitDegrees=45, twistLimitDegrees=1}
 {bone1Name=dyn_flap_02, bone2Name=dyn_bag_MSTR, swing1LimitDegrees=1, swing2LimitDegrees=15, twistLimitDegrees=1}
 {
  bone1Name=dyn_ball_01
  bone2Name=dyn_bag_MSTR
  swing1LimitDegrees=45
  swing1Motion=Free
  swing2LimitDegrees=45
  swing2Motion=Free
  twistLimitDegrees=45
  twistMotion=Free
 }
 {
  bone1Name=dyn_ball_02
  bone2Name=dyn_bag_MSTR
  swing1LimitDegrees=45
  swing1Motion=Free
  swing2LimitDegrees=45
  swing2Motion=Free
  twistLimitDegrees=45
  twistMotion=Free
 }
 {
  bone1Name=dyn_ball_03
  bone2Name=dyn_bag_MSTR
  swing1LimitDegrees=45
  swing1Motion=Free
  swing2LimitDegrees=45
  swing2Motion=Free
  twistLimitDegrees=45
  twistMotion=Free
 }
 {
  bone1Name=dyn_ball_04
  bone2Name=dyn_bag_MSTR
  swing1LimitDegrees=45
  swing1Motion=Free
  swing2LimitDegrees=45
  swing2Motion=Free
  twistLimitDegrees=45
  twistMotion=Free
 }
]
